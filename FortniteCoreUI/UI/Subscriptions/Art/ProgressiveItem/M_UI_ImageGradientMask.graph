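# UI material in the subscriptions progressive item art that masks an image's alpha with a vertical gradient.
# The gradient comes from the UV's V axis, offset and scaled by Range Offset and Range Multiply, then saturated.
# IsBottomAligned blends between the gradient and its inverted counterpart, defaulting to the inverted one, which fades out toward the bottom.
# The result multiplies the alpha of the Texture parameter, which defaults to a backpack icon.
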Material /FortniteCoreUI/UI/Subscriptions/Art/ProgressiveItem/M_UI_ImageGradientMask
graph {
 properties {
  bUsedWithStaticMesh=false
  blendMode=BLEND_Translucent
  materialDomain=MD_UI
  pixelDepthOffsetMode=PDOM_Legacy
  usedShadingModels=Default Lit
 }
 nodes [
  n1 = TextureCoordinate()
  n2 = ScalarParameter() {defaultValue=0.125, parameterName=Range Offset, sliderMax=1}
  n3 = ScalarParameter() {defaultValue=1, parameterName=Range Multiply, sliderMax=10}
  n4 = Multiply(A=Subtract(A=ComponentMask(None=@n1) {g=true}, B=@n2), B=@n3)
  n5 = ComponentMask(None=TextureCoordinate()) {g=true}
  n6 = Multiply(A=Subtract(A=OneMinus(None=@n5), B=@n2), B=@n3)
  Multiply(
   A=TextureSampleParameter2D() {parameterName=Texture, texture=/Game/UI/Foundation/Textures/Icons/Backpacks/T-Icon-Backpacks-773-Carabus-L}.A
   B=LinearInterpolate(A=Saturate(None=@n4), B=Saturate(None=@n6), Alpha=ScalarParameter() {defaultValue=1, parameterName=IsBottomAligned})
  )
 ]
}
